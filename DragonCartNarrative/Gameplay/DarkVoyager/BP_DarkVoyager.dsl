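; Dark Voyager skeletal mesh actor that alternates its animation between hover and idle at random intervals.
; BeginPlay starts a timer to the idle state, and each state event clears the timer that led to it, then schedules the opposite state after a new random delay.
; Both events set the IsIdle flag on DV_HoverAnimBP to drive the animation change.
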(event EventBeginPlay
  (bind _returnvalue (Utilities|Time|SetTimerbyEvent (delegate Custom|SetIdleState) (Math|Random|RandomFloatinRange 2e+01 4e+01) true))
  (Variables|Default|SetAnimStateTimer _returnvalue))

(event Custom|SetHoverState
  (Utilities|Time|ClearandInvalidateTimerbyHandle (Variables|Default|GetIdleStateTimer))
  (bind _asdv_hover_anim_bp (Utilities|Casting|CastToDV_HoverAnimBP (Components|SkeletalMesh|GetAnimInstance (Variables|SkeletalMeshActor|GetSkeletalMeshComponent))))
  (Variables|/DragonCartNarrative/DV_HoverAnimBP.DV_HoverAnimBP_C|SetIsIdle false _asdv_hover_anim_bp)
  (bind _returnvalue (Utilities|Time|SetTimerbyEvent (delegate Custom|SetIdleState) (Math|Random|RandomFloatinRange 2e+01 4e+01) true))
  (Variables|Default|SetAnimStateTimer _returnvalue))

(event Custom|SetIdleState
  (Utilities|Time|ClearandInvalidateTimerbyHandle (Variables|Default|GetAnimStateTimer))
  (bind _asdv_hover_anim_bp (Utilities|Casting|CastToDV_HoverAnimBP (Components|SkeletalMesh|GetAnimInstance (Variables|SkeletalMeshActor|GetSkeletalMeshComponent))))
  (Variables|/DragonCartNarrative/DV_HoverAnimBP.DV_HoverAnimBP_C|SetIsIdle true _asdv_hover_anim_bp)
  (bind _returnvalue (Utilities|Time|SetTimerbyEvent (EventDispatchers|CreateEvent :!event Custom|SetHoverState) (Math|Random|RandomFloatinRange 2e+01 4e+01) true))
  (Variables|Default|SetIdleStateTimer _returnvalue))

(fn ConstructionScript)
