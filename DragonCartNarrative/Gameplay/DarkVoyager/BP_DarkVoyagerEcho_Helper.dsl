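; Helper actor for the Dark Voyager Echo boss that holds a reference to the boss pawn and exposes it through a getter.
; On BeginPlay it binds to its own OnDestroyed event.
; When the helper is destroyed, it clears the pawn reference so the boss can be garbage collected.
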(event EventBeginPlay
  ; Dark Voyager Echo Boss
  (Game|BindEventtoOnDestroyed (delegate Custom|OnDestroyed_Event)))

(event Custom|OnDestroyed_Event
  (params (DestroyedActor :type Actor%20Object%20Reference))
  ; Dark Voyager Echo Boss
  (Variables|DarkVoyagerEcho|SetDark%20Voyager%20Echo%20Pawn 0) ; clear this variable when destroyed so\nthe Dark Voyager Echo boss is\nallowed to garbage collect
)

(fn GetDarkVoyagerEchoPawnActor
  (return (Variables|DarkVoyagerEcho|GetDark%20Voyager%20Echo%20Pawn)))

(fn ConstructionScript)
